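# Lit material for the Season 10 Keving memorial monument: patinated metal with a fake night-time light.
# A masks texture drives a height blend between fresnel-tinted metal and patina, plus an overlay detail and a patina-dependent scalar mask.
# The fake light combines a downward normal-map dot product with an actor-relative sphere mask, shaped by a gradient texture and weakened on upward-facing surfaces.
# A time-of-day function restricts that light to night.
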Material /Figment_S10/Athena/Items/QuestInteractables/S10_KevingMemorial/M_Monument
graph {
 properties {pixelDepthOffsetMode=PDOM_Legacy, usedShadingModels=Default Lit}
 nodes [
  n1 = TextureSample() {samplerType=SAMPLERTYPE_Masks, texture=&/MetalCube_S}
  n2 = MaterialFunctionCall(
   A=VectorParameter() {parameterName=Patina Color, defaultValue=<rgba 0.1441 0.401 0.3319 0>}.RGB
   B=LinearInterpolate(
    A=VectorParameter() {parameterName=Metal Color, defaultValue=<rgba 0.3281 0.2991 0.2666 0>}.RGB
    B=VectorParameter() {parameterName=Metal Color Fresnel, defaultValue=<rgba 0.0214 0.0594 0.0625 0>}.RGB
    Alpha=Fresnel() {exponent=3}
   )
   Transition Phase=ScalarParameter() {defaultValue=0.3905, parameterName=Patina transition}
   Height Texture=@n1.G
   Contrast=ScalarParameter() {defaultValue=1, parameterName=Patina Contrast}
  ) {materialFunction=&HeightLerp}
  LinearInterpolate(A=Constant() {r=1}, B=Multiply(A=Constant() {r=0.55}, B=@n1.B), Alpha=@n2.Lerp Alpha No Contrast)
  MaterialFunctionCall(Base=@n2.Results, Blend=@n1.R) {materialFunction=&Blend_Overlay}
  Multiply(A=@n2.Lerp Alpha No Contrast, B=Constant() {r=1})
  n3 = Transform(None=TextureSample() {samplerType=SAMPLERTYPE_Normal, texture=/Game/Athena/Items/Consumables/Spooky_Mist_Cube/Textures/Spooky_Mist_Cube_N}.RGB)
  n4 = Constant3Vector() {constant=<rgba 0 0 -1 0>}
  n5 = Saturate(None=Power(Base=DotProduct(A=@n3, B=@n4), Exp=ScalarParameter() {defaultValue=4.104, parameterName=Fake Light Exponent}))
  n6 = VectorParameter() {parameterName=Fake Light Brightness, defaultValue=<rgba 2 1.7069 0.6979 0>}
  n7 = VertexNormalWS()
  n8 = SphereMask(
   A=WorldPosition().XYZ
   B=Subtract(A=ActorPositionWS(), B=VectorParameter() {parameterName=Light Location, defaultValue=<rgba -6.7705 20.5535 -192.4696 0>}.RGB)
   Radius=ScalarParameter() {defaultValue=218.6429, parameterName=Fake Light Radius}
   Hardness=ScalarParameter() {defaultValue=0.3, parameterName=Fake Light Focus}
  )
  n9 = TextureSample(UVs=@n8) {texture=/Figment_S10/Characters/Enemies/Cattus_Head/Textures/T_IceGradient_Cattus}
  n10 = Multiply(A=OneMinus(None=ComponentMask(None=@n7) {b=true}), B=Multiply(A=@n8, B=OneMinus(None=@n9.B)))
  n11 = Constant()
  Multiply(
   A=Multiply(A=Multiply(A=@n5, B=@n6.RGB), B=Saturate(None=@n10))
   B=MaterialFunctionCall(Color_Morning=@n11, Color_Day=@n11, Color_Evening=@n11, Color_Night=Constant() {r=1}) {materialFunction=&Color_Time_of_Day}.Result
  )
 ]
}
